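# Looping idle Niagara effect for the RelayStick Plume emote prop, attached to bones of a skeletal mesh.
# Three mesh rings with glows lag behind owner velocity, alongside three lights, three soft ribbon trails and bone sparkles.
# Base dust and center sparkles add motion, with the sparkles spawning child dust through particle reads.
# System audio amplitude, color phase and spawn alpha drive scale, color and glow, with time-of-day tinting on rings and bone sparkles.
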NiagaraSystem /BRCosmetics/Animation/Game/MainPlayer/Emotes/RelayStick_Plume/FX/NS_RelayStick_Plume_IdleLoop
dependencies {
 usedDataInterfaces [
  &NiagaraDataInterfaceCurve
  &NiagaraDataInterfaceParticleRead
  &NiagaraDataInterfaceArrayInt32
  &NiagaraDataInterfaceArrayFloat
  &NiagaraDataInterfaceArrayFloat3
  &NiagaraDataInterfaceSkeletalMesh
  &NiagaraDataInterfaceColorCurve
 ]
 usedDynamicInputs [
  null
  &MakeVector2D
  &Power
  &SetBoolByFloatComparison
  &InvertBool
  &RemapRange
  &Boolean_OrOperation
  &LerpLinearColors
  &ColorFromCurve
 ]
 usedModules [
  &self:SystemSpawnScript.SystemScriptSource.SystemScriptGraph.NiagaraNodeAssignment_0.SetVariables_#1
  &self:SystemSpawnScript.SystemScriptSource.SystemScriptGraph.NiagaraNodeAssignment_3.SetVariables_#2
  &SystemState
  &self:SystemSpawnScript.SystemScriptSource.SystemScriptGraph.NiagaraNodeAssignment_6.SetVariables_#3
  &self:SystemSpawnScript.SystemScriptSource.SystemScriptGraph.NiagaraNodeAssignment_2.SetVariables_#4
  &Timeline
  &self:SystemSpawnScript.SystemScriptSource.SystemScriptGraph.NiagaraNodeAssignment_9.SetVariables_#5
  &self:SystemSpawnScript.SystemScriptSource.SystemScriptGraph.NiagaraNodeAssignment_1.SetVariables_#6
  &self:SystemSpawnScript.SystemScriptSource.SystemScriptGraph.NiagaraNodeAssignment_4.SetVariables_#7
 ]
}
emitters {
 NE_BaseRings {
  emitterUpdateScript [
   &EmitterState(
    Inactive Response=::NewEnumerator1
    Life Cycle Mode=::NewEnumerator1
    Loop Behavior=::NewEnumerator1
    Loop Duration Mode=::NewEnumerator1
    UseLoopDelay=false
   )
   &SpawnBurst_Instantaneous(Spawn Count=3, Use Spawn Probability=false, UseLoopCountLimit=false)
  ]
  particleSpawnScript [
   &FN_SetSpawnGroup(Number of Spawn Groups=3)
   &InitializeParticle(
    Lifetime=5
    Mesh Scale=&SelectVectorFromArray(
     Direct Array Index=&SpawnGroup()
     Vector Selection Array=DataInterface() {propertyValues {FloatData [<1 1.25 1>, <0.8 0.8 0.8>, <0.6 0.5 0.6>]}}
    )
    Mesh Scale Mode=::NewEnumerator4
    Position Offset=<0 7 49>
    Sprite Size Mode=::NewEnumerator3
    Uniform Sprite Size=&SelectFloatFromArray(Direct Array Index=&SpawnGroup(), Float Selection Array=DataInterface() {propertyValues {FloatData [25,25,35]}})
    UsePositionOffset=true
    Write Material Random=true
   )
   &InitialMeshOrientation(
    Mesh Orientation Mode=::NewEnumerator7
    Rotation=&SelectVectorFromArray(
     Direct Array Index=&SpawnGroup()
     Vector Selection Array=DataInterface() {propertyValues {FloatData [<0 0 0>, <0 -0.03 0>, <0 -0.07 0>]}}
    )
    Use Rotation Vector=true
   )
   &SkeletalMeshLocation(
    Bone / Socket Index=5
    Bone Sampling Mode=::NewEnumerator4
    MaskSpawnGroup=false
    Sampled Position Offset=&Add_Vector(
     A=&SelectVectorFromArray(
      Direct Array Index=&SpawnGroup()
      Vector Selection Array=DataInterface() {propertyValues {FloatData [<-25 -1 -20>, <-25 -1 -15>, <-25 -1 -10>]}}
     )
     B=&Multiply_VectorByFloat(
      Vector=<0 0 1>
      Float=&Sine(
       Normalized Angle=&Add_Float(A=$Emitter.Age, B=&Multiply_FloatByInt(Float=0.2523, Integer=&SpawnGroup()))
       Period=1.5
       Scale=0
       Bias=0
      )
     )
    )
    Skeletal Mesh=$System.Skeletal Mesh
   ) {module=SkeletalMeshLocation001}
  ]
  particleUpdateScript [
   &ParticleState(Kill Particles When Lifetime Has Elapsed=false)
   /SetVariables(Particles.SplineThickness:NiagaraFloat=&SelectFloatFromArray(Direct Array Index=&SpawnGroup(), Float Selection Array=DataInterface() {propertyValues {FloatData [7,5,3]}})) {isSetParameters=true}
   &InterpolateOverTime(
    Rate Of Change=5
    Target Vector Value:Vector3f=&Multiply_Vector(
     A=/Niagara/DynamicInputs/Math/Clamp_Vector(
      Vector:Vector3f=&TransformVector(
       Vector=&Multiply_VectorByFloat(Vector=$Engine.Owner.Velocity, Float=0.01)
       Source Space=::World
       Destination Space=::Local
      )
      Min:Vector3f=&VectorFromFloat(Value=-1)
      Max:Vector3f=&VectorFromFloat(Value=1)
     )
     B=&Multiply_VectorByFloat(
      Vector=&SelectVectorFromArray(
       Direct Array Index=&SpawnGroup()
       Vector Selection Array=DataInterface() {propertyValues {FloatData [<0 0 0>, <0 5 5>, <0 10 10>]}}
      )
      Float=-1
     )
    )
    Variable Type=::NewEnumerator2
   )
   &SkeletalMeshLocation(
    Bone / Socket Index=5
    Bone Sampling Mode=::NewEnumerator4
    MaskSpawnGroup=false
    Position Sampling=::NewEnumerator2
    Sampled Position Offset=&Add_Vector(
     A=&SelectVectorFromArray(
      Direct Array Index=&SpawnGroup()
      Vector Selection Array=DataInterface() {propertyValues {FloatData [<2 0 0>, <1 0 0>, <0 0 0>]}}
     )
     B=<0 0 0>
    )
    Skeletal Mesh=$System.Skeletal Mesh
   )
   /SetVariables(Particles.GoalPos:NiagaraPosition=&AddVectorToPosition(Position=$Particles.SkeletalMeshLocation.SampledPosition, Vector=$Particles.InterpolateOverTime.InterpolateOverTime_V3_.Moving Average)) {isSetParameters=true}
   /SetVariables(Particles.Position=&TransformPosition(
     Position=&Lerp_Vector(
      A=Linked() {linkedVariable {name=Particles.Position, type=NiagaraPosition}}
      B=Linked() {linkedVariable {name=Particles.GoalPos, type=NiagaraPosition}}
      Alpha=&Multiply_Float(
       A=$Engine.DeltaTime
       B=&SelectFloatFromArray(Direct Array Index=&SpawnGroup(), Float Selection Array=DataInterface() {propertyValues {FloatData [25,10,8]}})
      )
     )
     Source Space=::Simulation
     Destination Space=::Simulation
    )) {isSetParameters=true}
   &DynamicMaterialParameters(
    Index 0 Param 1=&SelectFloatFromArray(
     Direct Array Index=&SpawnGroup()
     Float Selection Array=DataInterface() {propertyValues {FloatData [0.4,0.2,0.1]}}
    )
    Index 0 Param 2=20
    Index 0 Param 3=&Lerp_Float(
     A=$System.ReactiveSmoothedAmplitudeValue
     B=&FloatFromCurve(
      FloatCurve=DataInterface() {propertyValues {Curve {keys [
          {interpMode=RCIM_Cubic, leaveTangent=4.8905, tangentMode=RCTM_Break, time=0, value=0.15}
          <at 0.2 0.9>
          {interpMode=RCIM_Cubic, leaveTangent=-1.0585, tangentMode=RCTM_Break, time=0.3, value=0.9}
          {arriveTangent=-0.6609, tangentMode=RCTM_User, time=1, value=0.15}
         ]}}}
      CurveIndex=$System.NormalizedLoopAge
     )
     Alpha=$System.Timeline001.Float
    )
    Index 0 Param 4=$System.Timeline001.Float
    Module.Index0Param0Write=true
    Module.Index0Param1Write=true
    Module.Index0Param2Write=true
    Module.Index0Param3Write=true
   )
   /SetVariables(Particles.GlowPos:NiagaraPosition=&FN_FrontEndVector3Modifier(
     Gameplay=&AddVectorToPosition(
      Position=$Particles.Position
      Vector=&RotateVectorByQuaternion(
       Vector To Rotate=&SelectVectorFromArray(
        Direct Array Index=&SpawnGroup()
        Vector Selection Array=DataInterface() {propertyValues {FloatData [<-40 0 0>, <-25 0 0>, <-25 0 0>]}}
       )
       Quaternion=$Particles.MeshOrientation
      )
     )
     Front End=&AddVectorToPosition(
      Position=$Particles.Position
      Vector=&RotateVectorByQuaternion(
       Vector To Rotate=&SelectVectorFromArray(
        Direct Array Index=&SpawnGroup()
        Vector Selection Array=DataInterface() {propertyValues {FloatData [<-40 10 0>, <-30 0 0>, <-30 0 0>]}}
       )
       Quaternion=$Particles.MeshOrientation
      )
     )
    )) {isSetParameters=true}
   /SetVariables(Particles.GlowDynMat:Vector4f=&MakeFloat4FromVectorAndFloat(
     "Vector (RGB)"=&RotateVectorByQuaternion(
      Vector To Rotate=<1 0 0>
      Quaternion=&MultiplyQuaternion(Quaternion A=$Particles.MeshOrientation, Quaternion B=$Engine.Owner.Rotation)
     )
     "Float (Alpha)"=&Multiply_Float(
      A=&SelectFloatFromArray(Direct Array Index=&SpawnGroup(), Float Selection Array=DataInterface() {propertyValues {FloatData [80,70,65]}})
      B=&Lerp_Float(
       A=1
       B=1.1
       Alpha=&Lerp_Float(A=$System.AudioWeight, B=$System.ReactiveSmoothedAmplitudeValue, Alpha=$System.Timeline001.Float)
      )
     )
    )) {isSetParameters=true}
   &ScaleMeshSize(Scale Factor=&Add_Vector(
     A=<1 1 1>
     B=&Multiply_VectorByFloat(
      Vector=<0.3 1 -0.4>
      Float=&Lerp_Float(
       A=0
       B=1
       Alpha=&MakeCustomFloatFromBool(Bool=$System.bIsReactive, True Float=$System.SmoothedAmplitudeValue, False Float=$System.AudioWeight)
      )
     )
    )) {module=ScaleMeshSize001}
   &ScaleColor(
    Scale Mode=::NewEnumerator1
    Scale RGBA=&MakeVector4_FromLinearColor(LinearColor=&MultiplyLinearColor(
      Linear Color=&FNTimeOfDayColor(Night Color=<rgba 0.4065 0.3505 0.7812 1>, Morning Color=<rgba 1 0.6536 0.6237 1>, Day Color=<rgba 1 1 1 1>, Evening Color=<rgba 1 1 1 1>)
      Scale Factor=&MakeFloat4FromVectorAndFloat("Vector (RGB)"=&MakeVectorFromLinearColorRGB(Color=<rgba 1 1 1 1>), "Float (Alpha)"=$System.SpawnAlpha)
     ))
    ScaleRGBA=true
   )
   &FN_Light_Attributes(
    Color=$System.ColorPhase
    Create Color Attribute=true
    Create Exponent Attribute=true
    Create Radius Attribute=true
    Enable Particle Light Enable=true
    Exponent=3
    Light Enabled:NiagaraBool=&Boolean_AndOperation(
     Bool1=&Boolean_NotOperation(Bool=$User.bIsFrontEndPreview)
     Bool2=&SetBoolByIntComparison(A=&SpawnGroup(), B=2, Comparison Type=::NewEnumerator4)
    )
    Radius=&Lerp_Float(A=65, B=85, Alpha=$System.SmoothedAmplitudeValue)
    Scale Color=4
   )
  ]
  rendererClasses [
   &NiagaraMeshRendererProperties
   &NiagaraSpriteRendererProperties
   &NiagaraLightRendererProperties
  ]
  renderers [
   {rendererClass=&NiagaraMeshRendererProperties, rendererIndex=0}
   {rendererClass=&NiagaraSpriteRendererProperties, rendererIndex=1}
   {rendererClass=&NiagaraLightRendererProperties, rendererIndex=2}
  ]
 }
 NE_BaseSmallDust {
  emitterUpdateScript [
   &EmitterState()
   &SpawnRate(SpawnRate=8, Use Spawn Probability=false)
  ]
  particleSpawnScript [
   &InitializeParticle(
    Color=$System.ColorPhase
    Lifetime Max=1
    Lifetime Min=0.6
    Lifetime Mode=::NewEnumerator1
    Sprite Size Mode=::NewEnumerator1
    Uniform Sprite Size Max=15
    Uniform Sprite Size Min=10
    UsePositionOffset=false
    Write Material Random=true
   )
   &SkeletalMeshLocation(
    Bone / Socket Index=5
    Bone Sampling Mode=::NewEnumerator4
    MaskSpawnGroup=false
    Sampled Position Offset=&RandomRangeVector(Minimum=<0 -5 -10>, Maximum=<-50 5 -10>)
    Skeletal Mesh=$System.Skeletal Mesh
   ) {module=SkeletalMeshLocation001}
  ]
  particleUpdateScript [
   &ParticleState()
   &ScaleSpriteSize(Uniform Curve Sprite Scale=DataInterface() {propertyValues {Curve {keys [
        {interpMode=RCIM_Cubic, tangentMode=RCTM_User, time=0, value=0}
        {interpMode=RCIM_Cubic, tangentMode=RCTM_User, time=0.5, value=1}
        {tangentMode=RCTM_User, time=1, value=0}
       ]}}})
   &InheritSourceMovement(
    Applied Position Scale=&VectorFromFloat(Value=&FloatFromCurve(FloatCurve=DataInterface() {propertyValues {Curve {keys [
          {arriveTangent=-2.5982, interpMode=RCIM_Cubic, leaveTangent=-2.5982, tangentMode=RCTM_User, time=0, value=1}
          {interpMode=RCIM_Cubic, leaveTangent=-8.0089, time=1, value=0}
         ]}}}))
    ApplyForce=false
    ApplyPosition=true
    ApplyPositionFalloff=false
   )
   &GravityForce(Gravity=<0 0 -85>)
   &Drag(Drag=3, Use Linear Drag=true, Use Rotational Drag=true)
   &CurlNoiseForce(Noise Strength=50, Pan Noise=false)
   &SolveForcesAndVelocity(Clamp Velocity=false, Limit Acceleration=false)
   &ScaleColor(
    Scale RGB=&VectorFromFloat(Value=&Multiply_Float(
      A=&RandomRangeFloat(Minimum=0.4, Maximum=0.65, Evaluation Type=::NewEnumerator1)
      B=&FloatFromCurve(FloatCurve=DataInterface() {propertyValues {Curve {keys [{tangentMode=RCTM_User, time=0, value=1}, {tangentMode=RCTM_User, time=1, value=0}]}}})
     ))
    ScaleA=true
    ScaleRGB=true
   )
  ]
  rendererClasses [&NiagaraSpriteRendererProperties]
  renderers [{rendererClass=&NiagaraSpriteRendererProperties}]
 }
 NE_CenterSparkle {
  emitterUpdateScript [
   &EmitterState()
   &SpawnRate(SpawnRate=15, Use Spawn Probability=false)
  ]
  particleSpawnScript [
   &InitializeParticle(
    Lifetime Max=0.65
    Lifetime Min=0.25
    Lifetime Mode=::NewEnumerator1
    Mass Mode=::NewEnumerator3
    Sprite Rotation Mode=::NewEnumerator2
    Sprite Size Mode=::NewEnumerator1
    Uniform Sprite Size Min=8
    UsePositionOffset=false
    Write Material Random=true
   )
   &ShapeLocation(MaskSpawnGroup=false, Override Seed=false, Sphere Radius=12.4991, UseOwnerScale=true)
   &SkeletalMeshLocation(
    Bone / Socket Index=&RandomRangeInt(Minimum=0, Maximum=4)
    Bone Sampling Mode=::NewEnumerator4
    MaskSpawnGroup=false
    Skeletal Mesh=$System.Skeletal Mesh
   )
   &AddVelocity(
    Module.Use Falloff Distance=false
    Override Seed=false
    Velocity Mode=::NewEnumerator1
    Velocity Origin=$Particles.Position
    Velocity Speed=&RandomRangeFloat(Minimum=25, Maximum=50)
   )
   &InheritVelocity(Inherited Velocity Amount Scale=<1 1 0.1>, Inherited Velocity Speed Limit=50, Speed Limit=true, UseSpeedThreshold=false)
   /SetVariables(Particles.Color=$System.ColorPhase) {isSetParameters=true}
  ]
  particleUpdateScript [
   &ParticleState()
   &ScaleSpriteSize(Uniform Curve Sprite Scale=DataInterface() {propertyValues {Curve {keys [
        {interpMode=RCIM_Cubic, tangentMode=RCTM_User, time=0, value=0}
        {interpMode=RCIM_Cubic, tangentMode=RCTM_User, time=0.5, value=1}
        {tangentMode=RCTM_User, time=1, value=0}
       ]}}})
   &ScaleSpriteSize(
    Scale Sprite Size Mode=::NewEnumerator0
    Uniform Scale Factor=&Waveform(
     Waveform Count=::NewEnumerator1
     Global Phase=&Add_Float(A=$Emitter.Age, B=&RandomRangeFloat(Minimum=0, Maximum=1))
     "Waveform [1]"=::NewEnumerator10
     "[1] Frequency"=5
     "[1] Amplitude Min/Max"=<0 1>
     "[1] Scale"=3
     "[2] Average"=false
     "Waveform [2]"=::NewEnumerator12
     "[2] Input Position"=0
     "[2] Phase"=$Emitter.Age
     "[2] Phase Scale"=1
     "[2] Compound Frequency"=2
     "[2] Frequency"=2
     "[2] Amplitude Min/Max"=<0 1>
     "[2] Scale"=1
     "[2] Bias"=0
     "[2] Exponent"=1
     "[2] Pulse Duration"=0.1
     "[2] Clip Triangle Min/Max"=<0 0>
     "[2] Random Seed"=0
    )
   ) {module=ScaleSpriteSize001}
   &Drag(Drag=4, Use Linear Drag=true, Use Rotational Drag=false)
   &GravityForce(Gravity=&RandomRangeVector(Minimum=<0 0 -58.583>, Maximum=<0 0 25.5628>))
   &ScaleColor(
    Scale Alpha=&Waveform(
     Global Phase=&Add_Float(A=$Emitter.Age, B=&RandomRangeFloat(Minimum=0, Maximum=1))
     "Waveform [1]"=::NewEnumerator12
     "[1] Frequency"=9
     "[1] Amplitude Min/Max"=<0 1>
    )
    Scale RGB=<1.5 1.5 1.5>
    ScaleA=true
    ScaleRGB=true
   )
   &InheritSourceMovement(
    Applied Position Scale=&VectorFromFloat(Value=&FloatFromCurve(FloatCurve=DataInterface() {propertyValues {Curve {keys [<at 0 1 RCIM_Cubic>, {interpMode=RCIM_Cubic, leaveTangent=-8.0089, time=1, value=0}]}}}))
    ApplyForce=false
    ApplyPosition=true
    ApplyPositionFalloff=false
   )
   &SolveForcesAndVelocity(Clamp Velocity=false, Limit Acceleration=false)
   &SpriteRotationRate(Rotation Rate=&FloatFromCurve(
     FloatCurve=DataInterface() {propertyValues {Curve {keys [
         {interpMode=RCIM_Cubic, tangentMode=RCTM_User, time=0.2, value=0}
         {arriveTangent=3.4687, interpMode=RCIM_Cubic, leaveTangent=3.4687, tangentMode=RCTM_User, time=1, value=1}
        ]}}}
     Scale Curve=800
    ))
   &CameraOffset(Camera Offset Amount=20)
  ]
  rendererClasses [&NiagaraSpriteRendererProperties]
  renderers [{rendererClass=&NiagaraSpriteRendererProperties}]
 }
 NE_ChildDust {
  emitterUpdateScript [
   &EmitterState()
   &SpawnParticlesFromOtherEmitter(
    Attribute Reader=DataInterface() {propertyValues {EmitterBinding {bindingMode=Other, emitterName=NE_CenterSparkle}}}
    Spawn Rate Per Particle Cap=500
    Use Maximum Age=false
    Use Minimum Age=false
   )
  ]
  particleSpawnScript [
   &InitializeParticle(
    Color=&RandomRangeLinearColor(Minimum=<rgba 0.4886 0.5173 1 1>, Maximum=<rgba 0.1968 0.249 1 1>)
    Lifetime Max=0.5
    Lifetime Min=0.3
    Lifetime Mode=::NewEnumerator1
    Sprite Size Mode=::NewEnumerator1
    Uniform Sprite Size Max=15
    Uniform Sprite Size Min=10
    UsePositionOffset=false
    Write Material Random=true
   )
   &SampleParticlesFromOtherEmitter(Color Sampling=::NewEnumerator0)
   &ShapeLocation(
    Hemisphere Distribution=<1 0.5>
    MaskSpawnGroup=false
    Override Seed=false
    Rotation Axis=<0 0 -1>
    Rotation Mode=::NewEnumerator1
    Sphere Radius=2
    UseOwnerScale=true
   )
   &AddVelocity(
    Module.Use Falloff Distance=false
    Override Seed=false
    Velocity Mode=::NewEnumerator1
    Velocity Origin=$Particles.Position
    Velocity Speed=35
   )
  ]
  particleUpdateScript [
   &ParticleState()
   &ScaleSpriteSize(Uniform Curve Sprite Scale=DataInterface() {propertyValues {Curve {keys [
        {interpMode=RCIM_Cubic, tangentMode=RCTM_User, time=0, value=0}
        {interpMode=RCIM_Cubic, tangentMode=RCTM_User, time=0.5, value=1}
        {tangentMode=RCTM_User, time=1, value=0}
       ]}}})
   &GravityForce(Gravity=<0 0 30>)
   &Drag(Drag=3, Use Linear Drag=true, Use Rotational Drag=true)
   &SolveForcesAndVelocity(Clamp Velocity=false, Limit Acceleration=false)
   &ScaleColor(
    Scale RGB=&VectorFromFloat(Value=&Multiply_Float(
      A=&RandomRangeFloat(Minimum=0.4, Maximum=0.65, Evaluation Type=::NewEnumerator1)
      B=&FloatFromCurve(FloatCurve=DataInterface() {propertyValues {Curve {keys [{tangentMode=RCTM_User, time=0, value=1}, {tangentMode=RCTM_User, time=1, value=0}]}}})
     ))
    ScaleA=true
    ScaleRGB=true
   )
  ]
  rendererClasses [&NiagaraSpriteRendererProperties]
  renderers [{rendererClass=&NiagaraSpriteRendererProperties}]
 }
 NE_LightAndGlow {
  emitterUpdateScript [
   &EmitterState(
    Inactive Response=::NewEnumerator1
    Life Cycle Mode=::NewEnumerator1
    Loop Behavior=::NewEnumerator1
    Loop Duration Mode=::NewEnumerator1
    UseLoopDelay=false
   )
   &SpawnBurst_Instantaneous(Spawn Count=3, Spawn Time=0.22, Use Spawn Probability=false, UseLoopCountLimit=false)
  ]
  particleSpawnScript [
   &InitializeParticle(Lifetime=5, Sprite Size Mode=::NewEnumerator3, Uniform Sprite Size=110, UsePositionOffset=true, Write Material Random=true)
   &PartitionParticles(Number Of Partitions=3)
  ]
  particleUpdateScript [
   &ParticleState(Kill Particles When Lifetime Has Elapsed=false)
   &ScaleSpriteSize(
    Scale Sprite Size Mode=::NewEnumerator0
    Uniform Scale Factor=&Lerp_Float(A=1, B=1.5, Alpha=$System.SmoothedAmplitudeValue)
   )
   &Color(
    Color=$System.ColorPhase
    Scale Alpha=$System.SpawnAlpha
    Scale Color=&VectorFromFloat(Value=$System.SpawnAlpha)
   )
   &SkeletalMeshLocation(
    Bone / Socket Index=&SelectIntFromArray(Direct Array Index=$Particles.PartitionParticles.Partition, Int Selection Array=DataInterface() {propertyValues {IntData [0,1,3]}})
    Bone Sampling Mode=::NewEnumerator4
    MaskSpawnGroup=false
    Sampled Position Offset Coordinate Space:ENiagaraSkelMeshTransforms=::NewEnumerator2
    Skeletal Mesh=$System.Skeletal Mesh
   )
   &ScaleColor(
    Color Value To Scale=$StackContext.Color
    Scale Alpha=&Multiply_Float(
     A=&FloatFromCurve(
      FloatCurve=DataInterface() {propertyValues {Curve {keys [{interpMode=RCIM_Cubic, tangentMode=RCTM_User, time=0, value=0.5}, <at 0.9 1 RCIM_Cubic>]}}}
      CurveIndex=$System.AudioWeight
     )
     B=$System.SpawnAlpha
    )
    Scale RGB=&VectorFromFloat(Value=1.5)
    ScaleA=true
    ScaleRGB=true
   )
   &DynamicMaterialParameters(
    Index 0 Param 1=&MakeFloatFromVector(
     Vector=&RotateVectorByQuaternion(Vector To Rotate=<0 1 0>, Quaternion=$Particles.SkeletalMeshLocation.SampledRotation)
     Channel=::NewEnumerator0
    )
    Index 0 Param 2=&MakeFloatFromVector(
     Vector=&RotateVectorByQuaternion(Vector To Rotate=<0 1 0>, Quaternion=$Particles.SkeletalMeshLocation.SampledRotation)
     Channel=::NewEnumerator1
    )
    Index 0 Param 3=&MakeFloatFromVector(
     Vector=&RotateVectorByQuaternion(Vector To Rotate=<0 1 0>, Quaternion=$Particles.SkeletalMeshLocation.SampledRotation)
     Channel=::NewEnumerator2
    )
    Index 0 Param 4=150
    Module.Index0Param0Write=true
    Module.Index0Param1Write=true
    Module.Index0Param2Write=true
    Module.Index0Param3Write=true
   ) {module=DynamicMaterialParameters002}
   &FN_Light_Attributes(
    Color=$Particles.Color
    Create Color Attribute=true
    Create Exponent Attribute=true
    Create Radius Attribute=true
    Exponent=1
    Radius=65
    Scale Color=&Multiply_Float(A=2, B=$System.SpawnAlpha)
   )
  ]
  rendererClasses [&NiagaraLightRendererProperties, &NiagaraSpriteRendererProperties]
  renderers [
   {rendererClass=&NiagaraLightRendererProperties, rendererIndex=0}
   {rendererClass=&NiagaraSpriteRendererProperties, rendererIndex=1}
  ]
 }
 NE_SoftPlumeTrails {
  emitterSpawnScript [
   /SetVariables(Emitter.RibbonNumber=3) {isSetParameters=true}
   /SetVariables(Emitter.ParticlePerRibbons=5) {isSetParameters=true}
  ]
  emitterUpdateScript [
   &EmitterState(
    Inactive Response=::NewEnumerator1
    Life Cycle Mode=::NewEnumerator1
    Loop Behavior=::NewEnumerator1
    Loop Duration Mode=::NewEnumerator1
    UseLoopDelay=false
   )
   &SpawnBurst_Instantaneous(
    Spawn Count=&Multiply_Int(A=$Emitter.ParticlePerRibbons, B=$Emitter.RibbonNumber)
    Use Spawn Probability=false
    UseLoopCountLimit=false
   )
  ]
  particleSpawnScript [
   &PartitionParticles(
    Number Of Partitions=$Emitter.RibbonNumber
    Particles Per Partition:NiagaraInt32=$Emitter.ParticlePerRibbons
    Partition Mode:ENiagara_PartitionMode=::NewEnumerator1
    Write Partition Index to Ribbon Link Order:ENiagara_AttributeSamplingApplyOutput=::NewEnumerator0
    Write Partition to Ribbon ID=::NewEnumerator0
   )
   &InitializeParticle(Lifetime=5, Ribbon Width=25, Ribbon Width Mode=::NewEnumerator1, UsePositionOffset=false, Write Material Random=true)
   &SkeletalMeshLocation(
    Bone / Socket Index=&SelectIntFromArray(Direct Array Index=$Particles.PartitionParticles.Partition, Int Selection Array=DataInterface() {propertyValues {IntData [0,2,3]}})
    Bone Sampling Mode=::NewEnumerator4
    MaskSpawnGroup=false
    Sampled Position Offset=&Lerp_Vector(
     A=<0 0 0>
     B=&Lerp_Vector(
      A=<-50 0 0>
      B=<-150 0 0>
      Alpha=&ClampFloat(Float=&Divide_Float(A=&VectorLength(Vector=$Engine.Owner.Velocity), B=500))
     )
     Alpha=$Particles.PartitionParticles.Normalized Index In Partition
    )
    Sampled Position Offset Coordinate Space:ENiagaraSkelMeshTransforms=::NewEnumerator0
    Skeletal Mesh=$System.Skeletal Mesh
   ) {module=SkeletalMeshLocation001}
  ]
  particleUpdateScript [
   &ParticleState(Kill Particles When Lifetime Has Elapsed=false)
   /SetVariables(Particles.RibbonU0Override=$Particles.PartitionParticles.Normalized Index In Partition) {isSetParameters=true}
   &SkeletalMeshLocation(
    Bone / Socket Index=&SelectIntFromArray(Direct Array Index=$Particles.PartitionParticles.Partition, Int Selection Array=DataInterface() {propertyValues {IntData [1,2,3]}})
    Bone Sampling Mode=::NewEnumerator4
    MaskSpawnGroup=false
    Position Sampling=::NewEnumerator2
    Sampled Position Offset=&Lerp_Vector(
     A=<0 0 0>
     B=&SelectVectorFromArray(
      Direct Array Index=$Particles.PartitionParticles.Partition
      Vector Selection Array=DataInterface() {propertyValues {FloatData [<35 -12 10>, <30 -15 10>, <20 -30 15>]}}
     )
     Alpha=$Particles.PartitionParticles.Normalized Index In Partition
    )
    Skeletal Mesh=$System.Skeletal Mesh
   )
   /SetVariables(Particles.RibbonFacing=&RotateVectorByQuaternion(Vector To Rotate=<0 -1 0>, Quaternion=$Particles.SkeletalMeshLocation.SampledRotation)) {isSetParameters=true}
   &InheritSourceMovement(
    Applied Position Falloff Distance=10
    Applied Position Scale=&Multiply_VectorByFloat(Vector=<1 1 1>, Float=-1)
    ApplyForce=false
    ApplyPosition=true
    ApplyPositionFalloff=false
    Source Velocity:Vector3f=&Clamp_Vector(
     Vector:Vector3f=$Engine.Owner.Velocity
     Min:Vector3f=&VectorFromFloat(Value=-150)
     Max:Vector3f=&VectorFromFloat(Value=150)
    )
   )
   /SetVariables(Particles.Position=&Lerp_Vector(
     A=$Particles.SkeletalMeshLocation.SampledPosition
     B=&Lerp_Vector(
      A=Linked() {linkedVariable {name=Particles.Position, type=NiagaraPosition}}
      B=$Particles.SkeletalMeshLocation.SampledPosition
      Alpha=&Multiply_Float(A=$Engine.DeltaTime, B=10)
     )
     Alpha=&FloatFromCurve(
      FloatCurve=DataInterface() {propertyValues {Curve {keys [
          {arriveTangent=4.4709, interpMode=RCIM_Cubic, leaveTangent=4.4709, tangentMode=RCTM_User, time=0, value=0}
          {arriveTangent=1.5197, interpMode=RCIM_Cubic, leaveTangent=1.5197, tangentMode=RCTM_User, time=0.5, value=1}
         ]}}}
      CurveIndex=$Particles.PartitionParticles.Normalized Index In Partition
     )
    )) {isSetParameters=true}
   &ScaleColor(
    Scale Alpha=&Multiply_Float(
     A=&FloatFromCurve(
      FloatCurve=DataInterface() {propertyValues {Curve {keys [<at 0 1>,<at 0.8 0>]}}}
      CurveIndex=$Particles.Initial.PartitionParticles.Normalized Index In Partition
     )
     B=$System.SpawnAlpha
    )
    Scale RGB=&MakeVectorFromLinearColorRGB(Color=$System.ColorPhase)
    ScaleA=true
    ScaleRGB=true
   )
   &SolveForcesAndVelocity(Clamp Velocity=false, Limit Acceleration=false)
  ]
  rendererClasses [&NiagaraRibbonRendererProperties]
  renderers [{rendererClass=&NiagaraRibbonRendererProperties}]
 }
 NE_Sparkles_Bones {
  emitterUpdateScript [
   &EmitterState()
   &SpawnRate(SpawnRate=15, Use Spawn Probability=true)
  ]
  particleSpawnScript [
   &InitializeParticle(
    Color=&MultiplyLinearColor(Linear Color=$System.ColorPhase, Scale Factor=&MakeVector4_FromFloat(Float=1.25))
    Lifetime Max=0.8
    Lifetime Min=0.6
    Lifetime Mode=::NewEnumerator1
    Sprite Rotation Mode=::NewEnumerator2
    Sprite Size Mode=::NewEnumerator3
    UsePositionOffset=false
    Write Material Random=true
   )
   &CameraOffset(Camera Offset Amount=5)
   &PartitionParticles(Number Of Partitions=5, Particles Per Partition=1, Partition Mode:ENiagara_PartitionMode=::NewEnumerator1)
  ]
  particleUpdateScript [
   &ParticleState()
   &ScaleColor(
    Scale Alpha=&FloatFromCurve(FloatCurve=DataInterface() {propertyValues {Curve {keys [
         {interpMode=RCIM_Cubic, tangentMode=RCTM_User, time=0, value=0}
         {interpMode=RCIM_Cubic, tangentMode=RCTM_User, time=0.5, value=1}
         {tangentMode=RCTM_User, time=1, value=0}
        ]}}})
    Scale RGB=&VectorFromFloat(Value=&FNTimeOfDayScalar(Night Scale=0.5, Morning Scale=1, Evening Scale=0.8))
    ScaleA=true
    ScaleRGB=true
   )
   &ScaleSpriteSize(
    Uniform Curve Scale=3
    Uniform Curve Sprite Scale=DataInterface() {propertyValues {Curve {keys [
        {interpMode=RCIM_Cubic, tangentMode=RCTM_User, time=0, value=0}
        {interpMode=RCIM_Cubic, tangentMode=RCTM_User, time=0.5, value=1}
        {tangentMode=RCTM_User, time=1, value=0}
       ]}}}
   )
   &SkeletalMeshLocation(
    Bone / Socket Index=&RandomRangeInt(Minimum=0, Maximum=4)
    Bone Sampling Mode=::NewEnumerator4
    MaskSpawnGroup=false
    Orientation Sampling=::NewEnumerator1
    Sampled Position Offset=&Add_Vector(A=<0 -5 0>, B=&RandomRangeVector(Minimum=<0 -10 10>, Maximum=<0 10 -10>))
    Skeletal Mesh=$System.Skeletal Mesh
   )
   &SolveForcesAndVelocity(Clamp Velocity=false, Limit Acceleration=false)
   &SpriteRotationRate(Rotation Rate=&FloatFromCurve(
     FloatCurve=DataInterface() {propertyValues {Curve {keys [
         {interpMode=RCIM_Cubic, tangentMode=RCTM_User, time=0.2, value=0}
         {arriveTangent=3.4687, interpMode=RCIM_Cubic, leaveTangent=3.4687, tangentMode=RCTM_User, time=1, value=1}
        ]}}}
     Scale Curve=800
    ))
  ]
  rendererClasses [&NiagaraSpriteRendererProperties]
  renderers [{rendererClass=&NiagaraSpriteRendererProperties}]
 }
}
summary {
 systemName=NS_RelayStick_Plume_IdleLoop
 emitters [
  {emitterName=NE_CenterSparkle, rendererClasses [&NiagaraSpriteRendererProperties]}
  {emitterName=NE_ChildDust, rendererClasses [&NiagaraSpriteRendererProperties]}
  {
   emitterName=NE_LightAndGlow
   rendererClasses [&NiagaraLightRendererProperties, &NiagaraSpriteRendererProperties]
  }
  {emitterName=NE_Sparkles_Bones, rendererClasses [&NiagaraSpriteRendererProperties]}
  {
   emitterName=NE_BaseRings
   rendererClasses [
    &NiagaraMeshRendererProperties
    &NiagaraSpriteRendererProperties
    &NiagaraLightRendererProperties
   ]
  }
  {emitterName=NE_SoftPlumeTrails, rendererClasses [&NiagaraRibbonRendererProperties]}
  {emitterName=NE_BaseSmallDust, rendererClasses [&NiagaraSpriteRendererProperties]}
 ]
 userVariables [
  User.bIsFrontEnd {type=NiagaraBool}
  User.bIsFrontEndPreview {type=NiagaraBool}
  User.Ranked_Curve {type=NiagaraFloat, defaultValue=0, description=""}
 ]
}
