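# Niagara effect for the spin moment of the RelayStick Plume emote, trailing light ribbons and sparkles from the skeletal mesh.
# Wide wing ribbons and thin light ribbons each split particles across four bone/socket indices, shifting from magenta-violet to cyan.
# GPU sparkle sprites spawn across five indices, inherit mesh velocity and flicker through waveforms.
# User.NormalizedNotifyProgress fades ribbons in, reduces wing velocity inheritance and tapers sparkle spawning to zero.
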NiagaraSystem /BRCosmetics/Animation/Game/MainPlayer/Emotes/RelayStick_Plume/FX/NS_RelayStick_Plume_OnSpin
dependencies {
 usedDataInterfaces [&NiagaraDataInterfaceCurve, &NiagaraDataInterfaceColorCurve, &NiagaraDataInterfaceSkeletalMesh]
 usedDynamicInputs []
 usedModules [
  &self:SystemSpawnScript.SystemScriptSource.SystemScriptGraph.NiagaraNodeAssignment_0.SetVariables_#1
  &SystemState
 ]
}
emitters {
 NE_ThinLightRibbons {
  emitterUpdateScript [
   &EmitterState()
   &SpawnRate(SpawnRate=100, Use Spawn Probability=false)
  ]
  particleSpawnScript [
   &InitializeParticle(Color Mode=::NewEnumerator0, Lifetime=1.5, UsePositionOffset=false, Write Material Random=true)
   &PartitionParticles(Number Of Partitions=4)
   &SkeletalMeshLocation(
    Bone / Socket Index=$Particles.PartitionParticles.Partition
    MaskSpawnGroup=false
    Mesh Sampling Type=::NewEnumerator5
    Orientation Sampling=::NewEnumerator1
    Sampled Orientation Coordinate Space=::World
    Skeletal Mesh=$System.Skeletal Mesh
    Socket Sampling Mode=::NewEnumerator5
   )
   /SetVariables(Particles.RibbonFacing=/Game/Effects/Niagara/DynamicInputs/Engine/FN_GetSystemAxis(Axis:ENiagaraOrientationAxis=::ZAxis, Flip Axis:NiagaraBool=false)) {isSetParameters=true}
  ]
  particleUpdateScript [
   &ParticleState()
   /SetVariables(Particles.RibbonU0Override=&Add_Float(
     A=&ReturnNormalizedExecIndex(Normalized Index Scale=2)
     B=&Multiply_Float(A=&Multiply_FloatByInt(Float=3, Integer=$Particles.PartitionParticles.Partition), B=1)
    )) {isSetParameters=true, module=SetVariables_#2}
   /SetVariables(Particles.RibbonWidth=&FloatFromCurve(
     FloatCurve=DataInterface() {propertyValues {Curve {keys [{arriveTangent=2.4236, interpMode=RCIM_Cubic, leaveTangent=2.4236, tangentMode=RCTM_User, time=0, value=0.5}, <at 1 5 RCIM_Cubic>]}}}
     Scale Curve=20
    )) {isSetParameters=true, module=SetVariables_#3}
   &GravityForce(Gravity=<0 0 -18.4146>)
   &Drag(Use Linear Drag=true, Use Rotational Drag=true)
   &SolveForcesAndVelocity(Clamp Velocity=false, Limit Acceleration=false)
   &Color(
    Color=&ColorFromCurve(ColorCurve=DataInterface() {propertyValues {
       AlphaCurve {keys [<at 0.0122 0>,<at 0.1431 1 RCIM_Cubic>,<at 0.5787 0 RCIM_Cubic>]}
       BlueCurve {keys [<at 0 2 RCIM_Cubic>,<at 0.592 1 RCIM_Cubic>]}
       GreenCurve {keys [<at 0 0.0102 RCIM_Cubic>,<at 0.592 0.7791 RCIM_Cubic>]}
       RedCurve {keys [<at 0 1.5913 RCIM_Cubic>,<at 0.592 0.0248 RCIM_Cubic>]}
      }})
    Scale Alpha=&FloatFromCurve(
     FloatCurve=DataInterface() {propertyValues {Curve {keys [
         {arriveTangent=9.0725, interpMode=RCIM_Cubic, leaveTangent=9.0725, tangentMode=RCTM_User, time=0, value=0}
         {arriveTangent=-2.5249, interpMode=RCIM_Cubic, leaveTangent=-2.5249, tangentMode=RCTM_User, time=0.1, value=1}
        ]}}}
     CurveIndex=$User.NormalizedNotifyProgress
     Scale Curve=0.25
    )
   )
   &ScaleColor(
    Color Value To Scale=$StackContext.Color
    Scale Alpha=0.2
    Scale RGB=&VectorFromFloat(Value=&FloatFromCurve(
      FloatCurve=DataInterface() {propertyValues {Curve {keys [
          {arriveTangent=-2.5249, interpMode=RCIM_Cubic, leaveTangent=-2.5249, tangentMode=RCTM_User, time=0, value=1}
          {arriveTangent=-0.0405, interpMode=RCIM_Cubic, leaveTangent=-0.0405, tangentMode=RCTM_User, time=1, value=0}
         ]}}}
      Scale Curve=5
     ))
    ScaleA=true
    ScaleRGB=true
   )
  ]
  rendererClasses [&NiagaraRibbonRendererProperties]
  renderers [{rendererClass=&NiagaraRibbonRendererProperties}]
 }
 NE_WingsRibbons {
  emitterUpdateScript [
   &EmitterState()
   &SpawnRate(SpawnRate=100, Use Spawn Probability=false)
  ]
  particleSpawnScript [
   &InitializeParticle(Color Mode=::NewEnumerator0, UsePositionOffset=false, Write Material Random=true)
   &PartitionParticles(Number Of Partitions=4)
   &SkeletalMeshLocation(
    Bone / Socket Index=$Particles.PartitionParticles.Partition
    MaskSpawnGroup=false
    Mesh Sampling Type=::NewEnumerator5
    Orientation Sampling=::NewEnumerator1
    Sampled Orientation Coordinate Space=::World
    Skeletal Mesh=$System.Skeletal Mesh
    Socket Sampling Mode=::NewEnumerator5
   )
   /SetVariables(Particles.RibbonFacing=&RotateVectorByQuaternion(Vector To Rotate=<0 0 -1>, Quaternion=$StackContext.SkeletalMeshLocation.SampledRotation)) {isSetParameters=true}
   &InheritVelocity(
    Inherited Velocity Amount Scale=&Multiply_VectorByFloat(
     Vector=<1 1 0>
     Float=&FloatFromCurve(FloatCurve=DataInterface() {propertyValues {Curve {keys [<at 0 1>,<at 1 0.5>]}}}, CurveIndex=$User.NormalizedNotifyProgress, Scale Curve=5)
    )
    Speed Limit=true
    UseSpeedThreshold=false
   )
  ]
  particleUpdateScript [
   &ParticleState()
   /SetVariables(Particles.RibbonU0Override=&Add_Float(
     A=&ReturnNormalizedExecIndex()
     B=&Multiply_Float(A=&Multiply_FloatByInt(Float=0.25, Integer=$Particles.PartitionParticles.Partition), B=1)
    )) {isSetParameters=true, module=SetVariables_#2}
   /SetVariables(Particles.RibbonWidth=&FloatFromCurve(
     FloatCurve=DataInterface() {propertyValues {Curve {keys [{arriveTangent=2.4236, interpMode=RCIM_Cubic, leaveTangent=2.4236, tangentMode=RCTM_User, time=0, value=0.5}, <at 1 0.875 RCIM_Cubic>]}}}
     Scale Curve=100
    )) {isSetParameters=true, module=SetVariables_#3}
   &GravityForce(Gravity=<0 0 -18.4146>)
   &Drag(Use Linear Drag=true, Use Rotational Drag=true)
   &SolveForcesAndVelocity(Clamp Velocity=false, Limit Acceleration=false)
   &Color(
    Color=&ColorFromCurve(ColorCurve=DataInterface() {propertyValues {
       AlphaCurve {keys [<at 0 0>,<at 0.0519 1 RCIM_Cubic>,<at 0.6214 0 RCIM_Cubic>]}
       BlueCurve {keys [<at 0 2 RCIM_Cubic>,<at 0.5075 1 RCIM_Cubic>]}
       GreenCurve {keys [<at 0 0.0102 RCIM_Cubic>,<at 0.5075 0.7791 RCIM_Cubic>]}
       RedCurve {keys [<at 0 1.5913 RCIM_Cubic>,<at 0.5075 0.0248 RCIM_Cubic>]}
      }})
    Scale Alpha=&FloatFromCurve(
     FloatCurve=DataInterface() {propertyValues {Curve {keys [
         {arriveTangent=9.0725, interpMode=RCIM_Cubic, leaveTangent=9.0725, tangentMode=RCTM_User, time=0, value=0}
         {arriveTangent=-2.5249, interpMode=RCIM_Cubic, leaveTangent=-2.5249, tangentMode=RCTM_User, time=0.1408, value=0.9847}
        ]}}}
     CurveIndex=$User.NormalizedNotifyProgress
     Scale Curve=0.25
    )
   )
  ]
  rendererClasses [&NiagaraRibbonRendererProperties]
  renderers [{rendererClass=&NiagaraRibbonRendererProperties}]
 }
 NE__Sparkles {
  simTarget=GPUComputeSim
  emitterUpdateScript [
   &EmitterState()
   &SpawnRate(
    SpawnRate=&FloatFromCurve(FloatCurve=DataInterface() {propertyValues {Curve {keys [<at 0 1>,<at 1 0>]}}}, CurveIndex=$User.NormalizedNotifyProgress, Scale Curve=150)
    Use Spawn Probability=false
   )
  ]
  particleSpawnScript [
   &InitializeParticle(
    Lifetime Max=2.25
    Lifetime Min=0.25
    Lifetime Mode=::NewEnumerator1
    Mass Mode=::NewEnumerator3
    Sprite Rotation Mode=::NewEnumerator2
    Sprite Size Mode=::NewEnumerator1
    Uniform Sprite Size Max=4
    Uniform Sprite Size Min=2
    UsePositionOffset=false
    Write Material Random=true
   )
   &PartitionParticles(Number Of Partitions=5)
   &SkeletalMeshLocation(
    Bone / Socket Index=$Particles.PartitionParticles.Partition
    MaskSpawnGroup=false
    Mesh Sampling Type=::NewEnumerator5
    Orientation Sampling=::NewEnumerator1
    Sampled Orientation Coordinate Space=::World
    Sampled Position Offset=&Add_Vector(A=<0 -5 0>, B=&RandomRangeVector(Minimum=<0 -10 10>, Maximum=<0 10 -10>))
    Skeletal Mesh=$System.Skeletal Mesh
    Socket Sampling Mode=::NewEnumerator5
   )
   /SetVariables(Particles.Color=&ColorFromCurve(
     ColorCurve=DataInterface() {propertyValues {
       AlphaCurve {keys [<at 0 1 RCIM_Cubic>,<at 1 1 RCIM_Cubic>]}
       BlueCurve {keys [<at 0 1 RCIM_Cubic>,<at 1 0.6734 RCIM_Cubic>]}
       GreenCurve {keys [<at 0 0.7791 RCIM_Cubic>,<at 1 0.1165 RCIM_Cubic>]}
       RedCurve {keys [<at 0 0.0248 RCIM_Cubic>,<at 1 1 RCIM_Cubic>]}
      }}
     CurveIndex=&RandomRangeFloat(Minimum=0, Maximum=1)
    )) {isSetParameters=true}
   &InheritVelocity(
    Inherited Velocity Amount Scale=<1 1 0.2>
    Inherited Velocity Speed Limit=200
    Speed Limit=true
    UseSpeedThreshold=false
    Velocity Source=$StackContext.SkeletalMeshLocation.SampledVelocity
   )
   &AddVelocity(Velocity=&RandomVector(Vector Scale=1), Velocity Speed Scale=0)
  ]
  particleUpdateScript [
   &ParticleState()
   &ScaleSpriteSize(
    Uniform Curve Scale=2
    Uniform Curve Sprite Scale=DataInterface() {propertyValues {Curve {keys [
        {interpMode=RCIM_Cubic, tangentMode=RCTM_User, time=0, value=0}
        {interpMode=RCIM_Cubic, tangentMode=RCTM_User, time=0.5, value=1}
        {tangentMode=RCTM_User, time=1, value=0}
       ]}}}
   )
   &ScaleSpriteSize(
    Scale Sprite Size Mode=::NewEnumerator0
    Uniform Scale Factor=&Waveform(
     Waveform Count=::NewEnumerator1
     Global Phase=&Add_Float(A=$Emitter.Age, B=&RandomRangeFloat(Minimum=0, Maximum=1))
     "Waveform [1]"=::NewEnumerator10
     "[1] Frequency"=5
     "[1] Amplitude Min/Max"=<0 1>
     "[1] Scale"=3
     "[2] Average"=false
     "Waveform [2]"=::NewEnumerator12
     "[2] Input Position"=0
     "[2] Phase"=$Emitter.Age
     "[2] Phase Scale"=1
     "[2] Compound Frequency"=2
     "[2] Frequency"=2
     "[2] Amplitude Min/Max"=<0 1>
     "[2] Scale"=1
     "[2] Bias"=0
     "[2] Exponent"=1
     "[2] Pulse Duration"=0.1
     "[2] Clip Triangle Min/Max"=<0 0>
     "[2] Random Seed"=0
    )
   ) {module=ScaleSpriteSize001}
   &Drag(Use Linear Drag=true, Use Rotational Drag=false)
   &GravityForce(Gravity=&RandomRangeVector(Minimum=<0 0 -117.5392>, Maximum=<0 0 0>))
   &ScaleColor(
    Scale Alpha=&Waveform(
     Global Phase=&Add_Float(A=$Emitter.Age, B=&RandomRangeFloat(Minimum=0, Maximum=1))
     "Waveform [1]"=::NewEnumerator12
     "[1] Frequency"=9
     "[1] Amplitude Min/Max"=<0 1>
    )
    Scale RGB=<1.5 1.5 1.5>
    ScaleA=true
    ScaleRGB=true
   )
   &SolveForcesAndVelocity(Clamp Velocity=false, Limit Acceleration=false)
   &SpriteRotationRate(Rotation Rate=&FloatFromCurve(
     FloatCurve=DataInterface() {propertyValues {Curve {keys [
         {interpMode=RCIM_Cubic, tangentMode=RCTM_User, time=0.2, value=0}
         {arriveTangent=3.4687, interpMode=RCIM_Cubic, leaveTangent=3.4687, tangentMode=RCTM_User, time=1, value=1}
        ]}}}
     Scale Curve=800
    ))
   &ScaleColor(ScaleA=true, ScaleRGB=true) {module=ScaleColor001}
   &CameraOffset(Camera Offset Amount=20)
  ]
  rendererClasses [&NiagaraSpriteRendererProperties]
  renderers [{rendererClass=&NiagaraSpriteRendererProperties}]
 }
}
summary {
 systemName=NS_RelayStick_Plume_OnSpin
 emitters [
  {emitterName=NE_WingsRibbons, rendererClasses [&NiagaraRibbonRendererProperties]}
  {emitterName=NE_ThinLightRibbons, rendererClasses [&NiagaraRibbonRendererProperties]}
  {emitterName=NE__Sparkles, simTarget=GPUComputeSim, rendererClasses [&NiagaraSpriteRendererProperties]}
 ]
 userVariables [User.NormalizedNotifyProgress {type=NiagaraFloat, defaultValue=0, description=""}]
}
